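# Physics asset for the Space Warrior backpack cosmetic mesh, giving its left and right tubes physics bodies for secondary motion.
# A larger capsule covers the pack bone, and small capsules cover the upper and lower dynamic tube joints on each side.
# Each of the four tube joints is constrained directly to the pack, not chained, with identical tight swing and twist limits.
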PhysicsAsset /BRCosmetics/Accessories/FORT_Backpacks/Backpack_M_MED_Space_Warrior/Meshes/M_MED_Space_Warrior_Pack_Physics
bodies {
 dyn_L_lowerTube_jnt [{shapeType=Capsule, length=2, radius=1.5, center=<1.234 -0.2242 -0.0424>, rotation=<rot -88.0319 179.9999 -179.9999>}]
 dyn_L_upperTube_jnt [{shapeType=Capsule, length=2, radius=1.5, center=<-0.7212 -0.2244 0.025>, rotation=<rot -88.0194 -179.9993 179.9993>}]
 dyn_R_lowerTube_jnt [{shapeType=Capsule, length=2, radius=1.5, center=<1.2342 -0.2242 -0.0393>, rotation=<rot -88.1773 -179.9999 179.9999>}]
 dyn_R_upperTube_jnt [{shapeType=Capsule, length=2, radius=1.5, center=<-0.7212 -0.2244 0.0252>, rotation=<rot -87.9995 179.9994 -179.9992>}]
 pack [{shapeType=Capsule, length=15, radius=5, center=<0 -1.2586 0>}]
}
constraints [
 {bone1Name=dyn_R_lowerTube_jnt, bone2Name=pack, swing1LimitDegrees=15, swing2LimitDegrees=10, twistLimitDegrees=5}
 {bone1Name=dyn_L_lowerTube_jnt, bone2Name=pack, swing1LimitDegrees=15, swing2LimitDegrees=10, twistLimitDegrees=5}
 {bone1Name=dyn_R_upperTube_jnt, bone2Name=pack, swing1LimitDegrees=15, swing2LimitDegrees=10, twistLimitDegrees=5}
 {bone1Name=dyn_L_upperTube_jnt, bone2Name=pack, swing1LimitDegrees=15, swing2LimitDegrees=10, twistLimitDegrees=5}
]
